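; Quest gameplay ability that tags the player's pawn when their vehicle exceeds a set speed.
; On the vehicle-entered event it checks the vehicle's tags against T_Car and starts a looping timer that polls forward km/h.
; Exceeding Max Speed adds loose gameplay tags and ends the ability; the exit event clears a valid timer, calls tag removal and ends it.
; Angle and degrees-turned helper functions are defined but unused here.
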(event Custom|Check%20For%20Speed
  ;[1 1 1] Check Speed
  (bind _asfort_athena_vehicle (Utilities|Casting|CastToFortAthenaVehicle (Vehicle|GetVehicleInterface (Variables|GAVars|Data|GetOwningPlayerPawn)))
    (:then
      (if (> (Vehicle|GetForwardSpeedKmh _asfort_athena_vehicle) (Variables|Default|GetMax%20Speed))
        (Ability|GameplayEffect|AddLooseGameplayTags (Variables|GAVars|Data|GetOwningPlayerPawn) (Variables|Default|GetGameplay%20Tags))
        (Ability|EndAbility)))
    (:CastFailed
      (Ability|EndAbility))))

(event Ability|EventActivateAbilityFromEvent
  (params (EventData :type Gameplay%20Event%20Data%20Structure%20%28by%20ref%29))
  (bind _timer_handle (Variables|Default|GetTimer%20Handle))
  (|CallParentFunction EventData)
  (switch Utilities|FlowControl|Switch|SwitchonGameplayTag (Utilities|Struct|BreakGameplayEventData EventData)
    (:Event.Vehicle.Entered
      ;[1 1 1] Check Vehicle Tags
      (if (GameplayTags|TagContainer|CompareTagAssetInterfacetoOtherTags (Variables|Default|GetT_Car) (Vehicle|GetVehicleAbilitySystemComponent (Vehicle|GetVehicleInterface (Variables|GAVars|Data|GetOwningPlayerPawn))))
        ;[1 1 1] Start Checking If Vehicle is Going Fast
        (bind _returnvalue (Utilities|Time|SetTimerbyEvent (delegate Custom|Check%20For%20Speed) (Variables|Default|GetExecutionTimeInterval) true))
        (Variables|Default|SetTimer%20Handle _returnvalue)
        (else
          (Ability|EndAbility))))
    (:Event.Vehicle.Exit
      ;[1 1 1] On Vehicle Exit Clear Timer and End Ability
      (if (Utilities|Time|IsValidTimerHandle _timer_handle)
        (Utilities|Time|ClearandInvalidateTimerbyHandle _timer_handle)
        (Ability|GameplayEffect|RemoveLooseGameplayTags 0 (Variables|Default|GetGameplay%20Tags))
        (else
          _))
      (Ability|EndAbility))))

(fn AngleBetweenVectors
  (params (Destination%20Rotation :type Vector) (Origin%20Rotation :type Vector) (Rotation%20Normal :type Vector))
  (return (Math|Float|SelectFloat (Math|Trig|RadiansToDegrees (Math|Trig|Acos(Radians) (/ (Math|Vector|DotProduct Origin%20Rotation Destination%20Rotation) (* (Math|Vector|VectorLength Destination%20Rotation) (Math|Vector|VectorLength Origin%20Rotation))))) (- 3.6e+02 (Math|Trig|RadiansToDegrees (Math|Trig|Acos(Radians) (/ (Math|Vector|DotProduct Origin%20Rotation Destination%20Rotation) (* (Math|Vector|VectorLength Destination%20Rotation) (Math|Vector|VectorLength Origin%20Rotation)))))) (> (Math|Vector|DotProduct Rotation%20Normal (Math|Vector|CrossProduct Origin%20Rotation Destination%20Rotation)) 0))))

(fn CalculateDegreesTurned
  (params (A :type Float%20%28double-precision%29) (B :type Float%20%28double-precision%29))
  (Variables|Default|SetMax%20Speed A))
